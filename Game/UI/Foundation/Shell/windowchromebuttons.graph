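# Window chrome widget, a CommonUserWidget whose only element is a WindowTitleBarArea with window buttons enabled, for the game's UI shell.
# The title bar area rests mostly transparent and reveals itself on hover.
# Mouse enter plays FadeIn and mouse leave plays FadeOut, two short Render Opacity animations between dimmed and fully opaque.
# PreConstruct, Construct and Tick events are empty.
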WidgetBlueprint /Game/UI/Foundation/Shell/windowchromebuttons
animations [
 FadeIn {
  binding_count=1
  display_label=""
  length_seconds=0.3
  object=&self:FadeIn
  sequence {bindings [WindowTitleBarArea_107 {
     id=#1
     possessed_class=WindowTitleBarArea
     tracks [{
       type=MovieSceneFloatTrack
       display_name=Render Opacity
       sections [{type=MovieSceneFloatSection, end_frame=6, start_frame=0, channels [None {type=MovieSceneScriptingFloatChannel, keys [[0,0.2],[6,1]]}]}]
      }]
    }]}
 }
 FadeOut {
  binding_count=1
  display_label=""
  length_seconds=0.35
  object=&self:FadeOut
  sequence {bindings [WindowTitleBarArea_107 {
     id=#1
     possessed_class=WindowTitleBarArea
     tracks [{
       type=MovieSceneFloatTrack
       display_name=Render Opacity
       sections [{type=MovieSceneFloatSection, end_frame=7, start_frame=0, channels [None {type=MovieSceneScriptingFloatChannel, keys [[0,1],[6.9997,0.2]]}]}]
      }]
    }]}
 }
]
graphs [EventGraph {nodes [
   Event OnMouseEnter()
   > PlayAnimation(InAnimation=Get FadeIn() -> (FadeIn:Widget Animation Object Reference))
   Event OnMouseLeave!() -> (OutputDelegate, then, MouseEvent:"Pointer Event Structure (by ref)") {is=title K2Node_Event}
   > PlayAnimation(InAnimation=Get FadeOut() -> (FadeOut:Widget Animation Object Reference))
   Event PreConstruct()
   Event Construct()
   Event Tick()
  ]}]
variables []
widgets {
 info {inheritedWidgetCount=0, parentClass=&CommonUserWidget, rootWidgetClass=/Script/UMG.WindowTitleBarArea, widgetCount=1}
 widgets [WindowTitleBarArea_107 {
   widgetClassPath=&WindowTitleBarArea
   bIsVariable=true
   widget=WindowTitleBarArea_107
   properties {bWindowButtonsEnabled=true, renderOpacity=0.2}
  }]
}
